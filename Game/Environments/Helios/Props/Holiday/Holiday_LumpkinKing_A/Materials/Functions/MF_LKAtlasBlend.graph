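# Material function for the Holiday Lumpkin King prop that recolors thorn and leg regions of incoming material attributes using curve-atlas gradients.
# Parameterized UV rectangles build a gradient per region from inverted V, driving curve lookups and base color blends.
# AO and normal lerp toward flat values, while emissive gets per-region intensity with optional mask modulation and additive boost.
# Outputs the modified attributes plus gradient and ceiled binary masks for both regions.
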MaterialFunction /Game/Environments/Helios/Props/Holiday/Holiday_LumpkinKing_A/Materials/Functions/MF_LKAtlasBlend
graph {nodes [
  n1 = GetMaterialAttributes(None=FunctionInput() {inputName=MA, inputType=FunctionInput_MaterialAttributes}) {attributeGetTypes [
    #1
    #2
    #3
    #4
   ]}
  n2 = Constant()
  n3 = MaterialFunctionCall(Float2=TextureCoordinate()) {materialFunction=&BreakOutFloat2Components}
  n4 = Constant() {r=1}
  n5 = Multiply(
   A=If(A=@n3.R, B=ScalarParameter() {group=C_AtlasColorBlend, parameterName=Thorns MinU, sortPriority=24}, "A > B"=@n4, "A == B"=@n4, "A < B"=@n2)
   B=If(
    A=@n3.R
    B=ScalarParameter() {defaultValue=1, group=C_AtlasColorBlend, parameterName=Thorns MaxU, sortPriority=25}
    "A > B"=@n2
    "A == B"=@n4
    "A < B"=@n4
   )
  )
  n6 = OneMinus(None=@n3.G)
  n7 = Saturate(None=MaterialFunctionCall(
    Input=@n6
    Input Low=ScalarParameter() {group=C_AtlasColorBlend, parameterName=Thorns MinV, sortPriority=26}
    Input High=ScalarParameter() {defaultValue=1, group=C_AtlasColorBlend, parameterName=Thorns MaxV, sortPriority=27}
   ) {materialFunction=&RemapValueRange}.Result)
  n8 = CurveAtlasRowParameter(CurveTime=NamedRerouteDeclaration(None=Multiply(A=@n5, B=@n7)) {name=thornsParam, nodeColor=<rgba 1 0 0.4687 1>}) {
   atlas=/Game/Environments/Helios/Props/Holiday/Holiday_LumpkinKing_A/Materials/Functions/Curves/T_LKCurveAtlas
   curve=/Game/Environments/Helios/Props/Holiday/Holiday_LumpkinKing_A/Materials/Functions/Curves/Albedo_Curve_01
   group=C_AtlasColorBlend
   parameterName=LK_ThornsCurve
   sortPriority=16
  }
  n9 = NamedRerouteUsage()
  n10 = ScalarParameter() {group=C_AtlasColorBlend, parameterName=Legs MinU, sortPriority=33}
  n11 = Constant()
  n12 = Constant() {r=1}
  n13 = ScalarParameter() {defaultValue=1, group=C_AtlasColorBlend, parameterName=Legs MaxU, sortPriority=34}
  n14 = ScalarParameter() {group=C_AtlasColorBlend, parameterName=Legs MinV, sortPriority=35}
  n15 = ScalarParameter() {defaultValue=1, group=C_AtlasColorBlend, parameterName=Legs MaxV, sortPriority=36}
  n16 = Multiply(
   A=Multiply(A=If(A=@n3.R, B=@n10, "A > B"=@n12, "A == B"=@n12, "A < B"=@n11), B=If(A=@n3.R, B=@n13, "A > B"=@n11, "A == B"=@n12, "A < B"=@n12))
   B=Saturate(None=MaterialFunctionCall(Input=@n6, Input Low=@n14, Input High=@n15) {materialFunction=&RemapValueRange}.Result)
  )
  n17 = CurveAtlasRowParameter(CurveTime=@n16) {
   atlas=&T_LKCurveAtlas
   curve=/Game/Environments/Helios/Props/Holiday/Holiday_LumpkinKing_A/Materials/Functions/Curves/Albedo_Curve_02
   defaultValue=1
   group=C_AtlasColorBlend
   parameterName=LK_LegsCurve
  }
  n18 = If(A=@n6, B=@n15, "A > B"=Constant(), "A == B"=@n16, "A < B"=@n16)
  n19 = NamedRerouteUsage()
  n20 = SetMaterialAttributes!(
   MaterialAttributes=@n1.MaterialAttributes
   BaseColor=LinearInterpolate(
    A=LinearInterpolate(A=@n1.BaseColor, B=@n8, Alpha=@n9)
    B=@n17
    Alpha=NamedRerouteDeclaration(None=@n18) {name=legsParam, nodeColor=<rgba 0 0.7891 1 1>}
   )
   AmbientOcclusion=LinearInterpolate(A=@n1.AmbientOcclusion, B=Constant() {r=1}, Alpha=@n19)
   Normal=LinearInterpolate(A=@n1.Normal, B=Constant3Vector() {constant=<rgba 0 0 1 0>}, Alpha=@n19)
   EmissiveColor=NamedRerouteUsage()
  ) {attributeSetTypes [
    #1
    #2
    #3
    #4
   ]}
  FunctionOutput(None=@n20)
  FunctionOutput(None=NamedRerouteUsage()) {outputName=ThornsParam, sortPriority=1}
  FunctionOutput(None=NamedRerouteUsage()) {outputName=LegsParam, sortPriority=1}
  FunctionOutput(None=NamedRerouteUsage()) {outputName=ThornsCeil, sortPriority=2}
  FunctionOutput(None=NamedRerouteUsage()) {outputName=LegsCeil, sortPriority=3}
  NamedRerouteDeclaration(None=@n1.EmissiveColor) {name=inEmissive, nodeColor=<rgba 0.1484 1 0 1>}
  n21 = Power(
   Base=NamedRerouteUsage()
   Exp=ScalarParameter() {defaultValue=1, group=C_AtlasColorBlend, parameterName=LK_EmissiveContrast, sortPriority=50}
  )
  n22 = ScalarParameter() {defaultValue=1, group=C_AtlasColorBlend, parameterName=LK_ThornsEmissive, sortPriority=48}
  n23 = StaticSwitchParameter(True=Multiply(A=NamedRerouteDeclaration(None=@n21) {name=thornsMask, nodeColor=<rgba 1 0.9844 0 1>}, B=@n22), False=@n22) {group=C_AtlasColorBlend, parameterName=LK_ModulateEmissive, sortPriority=255}
  n24 = NamedRerouteUsage()
  n25 = Ceil(None=NamedRerouteUsage())
  n26 = NamedRerouteUsage()
  n27 = ScalarParameter() {defaultValue=1, group=C_AtlasColorBlend, parameterName=LK_EmissiveContrast, sortPriority=50}
  n28 = ScalarParameter() {defaultValue=1, group=C_AtlasColorBlend, parameterName=LK_LegsEmissive, sortPriority=48}
  n29 = Multiply(A=NamedRerouteDeclaration(None=Power(Base=@n26, Exp=@n27)) {name=legsMask, nodeColor=<rgba 1 0.5775 0.0177 1>}, B=@n28)
  n30 = NamedRerouteDeclaration(None=Ceil(None=NamedRerouteUsage())) {name=legsCeil, nodeColor=<rgba 0.7578 1 0 1>}
  n31 = LinearInterpolate(
   A=LinearInterpolate(A=NamedRerouteUsage(), B=Multiply(A=@n23, B=@n24), Alpha=@n25)
   B=Multiply(
    A=StaticSwitchParameter(True=@n29, False=@n28) {group=C_AtlasColorBlend, parameterName=LK_ModulateEmissive, sortPriority=255}
    B=NamedRerouteUsage()
   )
   Alpha=@n30
  )
  n32 = Saturate(None=Add(A=NamedRerouteDeclaration(None=@n25) {name=thornsCeil, nodeColor=<rgba 0 0.4375 1 1>}, B=@n30))
  NamedRerouteDeclaration(None=StaticSwitchParameter(True=Multiply(A=@n31, B=@n32), False=@n31) {group=C_AtlasColorBlend, parameterName=LK_ModulateEmissive, sortPriority=255}) {name=outEmissive, nodeColor=<rgba 1 0 0.7969 1>}
  n33 = NamedRerouteUsage()
  n34 = NamedRerouteUsage()
  n35 = NamedRerouteUsage()
  n36 = FunctionInput() {bUsePreviewValueAsDefault=true, inputName=EmissiveColor, sortPriority=1, previewValue=<1 1 1 1>}
  n37 = Multiply(
   A=ScalarParameter() {defaultValue=1, group=C_AtlasColorBlend, parameterName=LK_MaskAddBoost, sortPriority=50}
   B=Multiply(A=Add(A=@n34, B=@n35), B=@n36)
  )
  NamedRerouteDeclaration(None=StaticSwitchParameter(True=Add(A=@n33, B=@n37), False=NamedRerouteUsage()) {group=C_AtlasColorBlend, parameterName=LK_AddMask, sortPriority=255}) {name=maskedInEmissive, nodeColor=<rgba 0 1 0.3438 1>}
 ]}
