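# Material function for the FNM vine look on rift spawner CubeGrowth bases, built on LumpkinKing prop functions; the "FNM Vines" switch otherwise passes input attributes through.
# MF_LKLayer builds two layers, the first with fading triplanar detail, UV animation and masked vine color variation.
# MF_LKAtlasBlend optionally applies curve blends and supplies a thorn gradient for added emissive.
# An optional Fresnel-based inner-glow falloff then scales the emissive.
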MaterialFunction /Game/Athena/Deimos/Spawners/RiftSpawners/Assets/CubeGrowth/MF_SpawnerBase_Master_FNM
graph {nodes [
  n1 = NamedRerouteUsage()
  n2 = NamedRerouteUsage()
  n3 = NamedRerouteUsage()
  n4 = MaterialFunctionCall!(
   MA=@n1
   EmissiveColor=StaticSwitchParameter(True=StaticSwitchParameter(True=@n2, False=@n3) {group=C_AtlasColorBlend, parameterName=Use L1 EmissiveColor}, False=NamedRerouteUsage()) {group=Options, parameterName=Use Second Layer}
  ) {materialFunction=/Game/Environments/Helios/Props/Holiday/Holiday_LumpkinKing_A/Materials/Functions/MF_LKAtlasBlend}
  n5 = StaticSwitchParameter(True=@n4.Result, False=@n1) {group=Options, parameterName=Use CurveBlends}
  n6 = GetMaterialAttributes(None=@n5) {attributeGetTypes [#1]}
  n7 = NamedRerouteUsage()
  n8 = StaticSwitchParameter(True=NamedRerouteUsage(), False=NamedRerouteUsage()) {desc=Use AOSR instead, group=A_Layer_1, parameterName=L1_UseAOER, sortPriority=255}
  n9 = Multiply(
   A=Power(Base=NamedRerouteUsage(), Exp=ScalarParameter() {defaultValue=1, group=D_ThornsEmissive, parameterName=ThornsGradientContrast})
   B=Multiply(A=TextureSample(UVs=@n7, Tex=@n8).G, B=NamedRerouteUsage())
  )
  n10 = SetMaterialAttributes!(
   MaterialAttributes=@n6.MaterialAttributes
   EmissiveColor=LinearInterpolate(
    A=@n6.EmissiveColor
    B=Add(A=@n6.EmissiveColor, B=@n9)
    Alpha=ScalarParameter() {defaultValue=1, group=D_ThornsEmissive, parameterName=ThornsFixAmount}
   )
  ) {attributeSetTypes [#1]}
  n11 = NamedRerouteDeclaration(None=@n10) {name=Yumebau, nodeColor=<rgba 0 0.7656 1 1>}
  n12 = GetMaterialAttributes(None=@n11) {attributeGetTypes [#1]}
  n13 = SetMaterialAttributes(MaterialAttributes=@n12.MaterialAttributes, EmissiveColor=Multiply(A=@n12.EmissiveColor, B=NamedRerouteUsage())) {attributeSetTypes [#1]}
  n14 = FunctionInput() {inputName=Result, inputType=FunctionInput_MaterialAttributes}
  FunctionOutput(None=StaticSwitchParameter(
    True=StaticSwitchParameter(True=@n13, False=@n11) {group=Options, parameterName=UseLumpkinInnerFalloffGlow}
    False=NamedRerouteDeclaration(None=@n14) {name=MA_IN, nodeColor=<rgba 0.1484 1 0 1>}
   ) {parameterName=FNM Vines}) {outputName=SetMaterialAttributes}
  NamedRerouteDeclaration(None=TextureObjectParameter() {group=A_Layer_1, parameterName=L1_Albedo, sortPriority=16}) {name=L1Albedo, nodeColor=<rgba 1 0.0234 0 1>}
  NamedRerouteDeclaration(None=VectorParameter() {group=A_Layer_1, parameterName=L1 Emissive Color}.RGB) {name=L1EmColor, nodeColor=<rgba 0.9453 1 0 1>}
  n15 = AppendVector(
   A=ScalarParameter() {defaultValue=1, group=D_ThornsEmissive, parameterName=Thorns U Scale}
   B=ScalarParameter() {defaultValue=1, group=D_ThornsEmissive, parameterName=Thorns V Scale}
  )
  n16 = StaticSwitchParameter(True=TextureCoordinate() {coordinateIndex=1}, False=TextureCoordinate()) {desc=Use AOSR instead, group=D_ThornsEmissive, parameterName=ThornsFix_UseUV1, sortPriority=255}
  n17 = ScalarParameter() {group=A_Layer_1, parameterName=L1 U speed, sortPriority=64}
  n18 = ScalarParameter() {group=A_Layer_1, parameterName=L1 V speed, sortPriority=64}
  n19 = StaticBool()
  n20 = StaticSwitchParameter(True=StaticBool() {value=true}, False=@n19) {group=A_Layer_1, parameterName=L1_AnimUseNormals, sortPriority=255}
  NamedRerouteDeclaration(None=MaterialFunctionCall!(
    UV=Multiply(A=@n15, B=@n16)
    UVSpeed=AppendVector(A=@n17, B=@n18)
    UseVtxNormals=StaticSwitchParameter(True=@n20, False=@n19) {group=A_Layer_1, parameterName=L1_UseUVAnim, sortPriority=255}
    FlattenNormals=ScalarParameter() {group=A_Layer_1, parameterName=L1 AnimNormals Flatten, sortPriority=50}
    NormalsAmount=ScalarParameter() {group=A_Layer_1, parameterName=L1 AnimNormals Amount, sortPriority=50}
   ) {materialFunction=/Game/Environments/Helios/Props/Holiday/Holiday_LumpkinKing_A/Materials/Functions/MF_UVAnim}.Result) {name=thornsUVAnim_1, nodeColor=<rgba 0.4297 1 0 1>}
  n21 = Divide(A=PixelDepth().R, B=ScalarParameter() {defaultValue=1000, group=A_Layer_1, parameterName=L1_DetailMaxDistance, sortPriority=100})
  n22 = VectorParameter() {
   group=A_Layer_1
   parameterName=L1_DetailAlpha
   channelNames {a="", b=Normal Alpha, g=AOSR Alpha, r=Albedo Alpha}
   defaultValue=<rgba 1 1 1 1>
  }
  n23 = MaterialFunctionCall(X=@n22.Albedo Alpha, Y=@n22.Albedo Alpha, Z=@n22.Albedo Alpha, A=@n22.Albedo Alpha) {materialFunction=&MakeFloat4}
  n24 = Multiply(A=OneMinus(None=Saturate(None=@n21)), B=@n23.Result)
  NamedRerouteDeclaration(None=StaticSwitchParameter(
    True=StaticSwitchParameter(True=@n24, False=@n23.Result) {defaultValue=true, group=A_Layer_1, parameterName=L1_FadeDetailMaps, sortPriority=255}
    False=@n23.Result
   ) {group=A_Layer_1, parameterName=L1_UseDetailMaps, sortPriority=255}) {name=L1TriAlpha, nodeColor=<rgba 1 0 0.6562 1>}
  n25 = StaticSwitchParameter(
   True=NamedRerouteUsage()
   False=NamedRerouteDeclaration(None=VectorParameter() {group=A_Layer_2, parameterName=L2 Emissive Color}.RGB) {name=L2EmColor, nodeColor=<rgba 0.7344 0 1 1>}
  ) {group=C_AtlasColorBlend, parameterName=Use L1 EmissiveColor}
  n26 = NamedRerouteUsage()
  NamedRerouteDeclaration(None=MaterialFunctionCall(MA=NamedRerouteUsage(), EmissiveColor=StaticSwitchParameter(True=@n25, False=@n26) {group=Options, parameterName=Use Second Layer}) {materialFunction=&MF_LKAtlasBlend}.ThornsParam) {name=thornsGradient, nodeColor=<rgba 1 0.2344 0 1>}
  n27 = Divide(A=PixelDepth().R, B=ScalarParameter() {defaultValue=1000, group=A_Layer_1, parameterName=L1_DetailMaxDistance, sortPriority=100})
  n28 = VectorParameter() {
   group=A_Layer_1
   parameterName=L1_DetailAlpha
   channelNames {a="", b=Normal Alpha, g=AOSR Alpha, r=Albedo Alpha}
   defaultValue=<rgba 1 1 1 1>
  }
  n29 = MaterialFunctionCall(X=@n28.Albedo Alpha, Y=@n28.Albedo Alpha, Z=@n28.Albedo Alpha, A=@n28.Albedo Alpha) {materialFunction=&MakeFloat4}
  n30 = Multiply(A=OneMinus(None=Saturate(None=@n27)), B=@n29.Result)
  NamedRerouteDeclaration(None=StaticSwitchParameter(
    True=StaticSwitchParameter(True=@n30, False=@n29.Result) {defaultValue=true, group=A_Layer_1, parameterName=L1_FadeDetailMaps, sortPriority=255}
    False=@n29.Result
   ) {group=A_Layer_1, parameterName=L1_UseDetailMaps, sortPriority=255}) {name=L1TriAlpha_1, nodeColor=<rgba 1 0 0.6562 1>}
  n31 = GetMaterialAttributes(None=NamedRerouteUsage()) {attributeGetTypes [#1]}
  n32 = Constant3Vector()
  n33 = MaterialFunctionCall(
   Invert Fresnel=StaticSwitchParameter(True=StaticBool() {value=true}, False=StaticBool()) {group=Fresnel, parameterName=InvertFresnel, sortPriority=50}
   Power=ScalarParameter() {defaultValue=5, group=Fresnel, parameterName=Fresnel Exponent, sortPriority=50}
  ) {materialFunction=&Fresnel_Function}
  NamedRerouteDeclaration(None=SetMaterialAttributes(MaterialAttributes=@n31.MaterialAttributes, EmissiveColor=LinearInterpolate(A=@n32, B=@n31.EmissiveColor, Alpha=@n33.Result)) {attributeSetTypes [#1]}) {name=emissiveFresnelMA, nodeColor=<rgba 0.7578 0 1 1>}
  n34 = StaticBool() {value=true}
  n35 = StaticBool()
  NamedRerouteDeclaration(None=StaticSwitchParameter(
    True=StaticSwitchParameter(True=@n34, False=@n35) {desc=Use UV0 otherwise..., group=A_Layer_1, parameterName=L1_UseDetailAOSR, sortPriority=255}
    False=@n35
   ) {group=A_Layer_1, parameterName=L1_UseDetailMaps, sortPriority=255}) {name=L1UseDetailAOSR_1, nodeColor=<rgba 0 0.9766 1 1>}
  NamedRerouteDeclaration(None=TextureObjectParameter() {group=A_Layer_1, parameterName=L1_Albedo, sortPriority=16}) {name=L1Albedo_1, nodeColor=<rgba 1 0.0234 0 1>}
  n36 = TextureObjectParameter() {group=A_Layer_1, parameterName=L1_Normals, samplerType=SAMPLERTYPE_Normal, sortPriority=16, texture=/Engine/EngineMaterials/FlatNormal}
  NamedRerouteDeclaration(None=@n36) {name=L1Normal_1, nodeColor=<rgba 0 0.2969 1 1>}
  NamedRerouteDeclaration(None=TextureObjectParameter() {group=A_Layer_1, parameterName=L1_AOSR, sortPriority=16, texture=/MeshModelingToolsetExp/Textures/DefaultTexture}) {name=L1AOSR_1, nodeColor=<rgba 0.6172 1 0 1>}
  n37 = StaticBool() {value=true}
  n38 = StaticBool()
  NamedRerouteDeclaration(None=StaticSwitchParameter(
    True=StaticSwitchParameter(True=@n37, False=@n38) {desc=Use UV0 otherwise..., group=A_Layer_1, parameterName=L1_UseDetailAOSR, sortPriority=255}
    False=@n38
   ) {group=A_Layer_1, parameterName=L1_UseDetailMaps, sortPriority=255}) {name=L1UseDetailAOSR, nodeColor=<rgba 0 0.9766 1 1>}
  n39 = ScalarParameter() {defaultValue=2, group=E_InnerGlowFalloff, parameterName=Lumpkin_InnerGlowFalloff}
  n40 = Fresnel(ExponentIn=@n39) {baseReflectFraction=0}
  NamedRerouteDeclaration(None=LinearInterpolate(
    A=ScalarParameter() {defaultValue=2, group=E_InnerGlowFalloff, parameterName=Lumpkin_InnerglowBrightness_Max}
    B=ScalarParameter() {group=E_InnerGlowFalloff, parameterName=Lumpkin_InnerglowBrightness_Min}
    Alpha=Saturate(None=@n40)
   )) {name=Lumpkin_InnerglowFalloff, nodeColor=<rgba 0 1 0.1797 1>}
  n41 = TextureObjectParameter() {group=A_Layer_1, parameterName=L1_DetailNormals, samplerType=SAMPLERTYPE_Normal, sortPriority=64, texture=&FlatNormal}
  NamedRerouteDeclaration(None=@n41) {name=L1DNormals_1, nodeColor=<rgba 1 0 0.9375 1>}
  NamedRerouteDeclaration(None=TextureObjectParameter() {group=A_Layer_1, parameterName=L1_DetailAOSR, sortPriority=64, texture=&DefaultTexture}) {name=L1DAOSR_1, nodeColor=<rgba 0 1 0.7422 1>}
  n42 = TextureObjectParameter() {group=A_Layer_1, parameterName=L1_Normals, samplerType=SAMPLERTYPE_Normal, sortPriority=16, texture=&FlatNormal}
  NamedRerouteDeclaration(None=@n42) {name=L1Normal, nodeColor=<rgba 0 0.2969 1 1>}
  n43 = TextureCoordinate() {coordinateIndex=1}
  n44 = TextureCoordinate()
  n45 = TextureObject()
  n46 = NamedRerouteUsage()
  n47 = NamedRerouteUsage()
  n48 = VectorParameter() {group=A_Layer_2, parameterName=L2 Emissive Color}
  NamedRerouteDeclaration(None=MaterialFunctionCall!(
    UVs=StaticSwitchParameter(True=@n43, False=@n44) {desc=Use UV0 otherwise..., group=A_Layer_2, parameterName=L2_UseUV1, sortPriority=0}
    Tobj_Albedo=@n45
    Tobj_Normal=@n45
    Tobj_AOSR=StaticSwitchParameter(True=@n46, False=@n47) {defaultValue=true, desc=Use AOSR instead, group=A_Layer_2, parameterName=L2_UseAOER, sortPriority=255}
    TriplanarAlpha
    AllowTriplanar
    TriplanarSize
    Tobj_DetailNormal=@n45
    Tobj_DetailAOSR=@n45
    UseAOSRDetail
    UseUVDetailMaps
    UVDetailTileIndex
    UseSER
    UseAOER=NamedRerouteUsage()
    Emissive Color=NamedRerouteDeclaration(None=@n48.RGB) {name=L2EmColor_1, nodeColor=<rgba 0.7344 0 1 1>}
    SpecularValue=ScalarParameter() {defaultValue=0.5, group=A_Layer_2, parameterName=L2 Specular, sortPriority=48}
    UseMasksOnly=NamedRerouteUsage()
    UseUVAnimation
    AnimAOSRweights
    AnimatedUVs
   ) {materialFunction=/Game/Environments/Helios/Props/Holiday/Holiday_LumpkinKing_A/Materials/Functions/MF_LKLayer}.Result) {name=L2_MA, nodeColor=<rgba 0.9643 0.1427 1 1>}
  NamedRerouteDeclaration(None=TextureObjectParameter() {group=A_Layer_2, parameterName=L2_AOER, sortPriority=16}) {name=L2AOER, nodeColor=<rgba 1 0.0234 0 1>}
  n49 = StaticBool() {value=true}
  n50 = StaticBool()
  NamedRerouteDeclaration(None=StaticSwitchParameter(True=@n49, False=@n50) {defaultValue=true, desc=Use AOSR instead, group=A_Layer_2, parameterName=L2_UseAOEROnly, sortPriority=255}) {name=L2UseOnlyAOER, nodeColor=<rgba 0.4531 1 0 1>}
  NamedRerouteDeclaration(None=TextureObjectParameter() {group=A_Layer_1, parameterName=L1_AOER, sortPriority=16, texture=&DefaultTexture}) {name=L1AOER_1, nodeColor=<rgba 0.2026 0.2461 1 1>}
  NamedRerouteDeclaration(None=TextureObjectParameter() {group=A_Layer_2, parameterName=L2_AOSR, sortPriority=16}) {name=L2AOSR, nodeColor=<rgba 1 0.1978 0.8982 1>}
  NamedRerouteDeclaration(None=StaticSwitchParameter(True=@n49, False=@n50) {defaultValue=true, desc=Use AOSR instead, group=A_Layer_2, parameterName=L2_UseAOER, sortPriority=255}) {name=L2UseAOER, nodeColor=<rgba 0.3125 0 1 1>}
  n51 = ScalarParameter() {group=A_Layer_1, parameterName=L1 U speed, sortPriority=64}
  n52 = ScalarParameter() {group=A_Layer_1, parameterName=L1 V speed, sortPriority=64}
  n53 = StaticBool()
  n54 = StaticSwitchParameter(True=StaticBool() {value=true}, False=@n53) {group=A_Layer_1, parameterName=L1_AnimUseNormals, sortPriority=255}
  NamedRerouteDeclaration(None=MaterialFunctionCall(
    UV=NamedRerouteUsage()
    UVSpeed=AppendVector(A=@n51, B=@n52)
    UseVtxNormals=StaticSwitchParameter(True=@n54, False=@n53) {group=A_Layer_1, parameterName=L1_UseUVAnim, sortPriority=255}
    FlattenNormals=ScalarParameter() {group=A_Layer_1, parameterName=L1 AnimNormals Flatten, sortPriority=50}
    NormalsAmount=ScalarParameter() {group=A_Layer_1, parameterName=L1 AnimNormals Amount, sortPriority=50}
   ) {materialFunction=&MF_UVAnim}.Result) {name=L1animatedUVs, nodeColor=<rgba 0.9453 0 1 1>}
  n55 = VectorParameter() {desc=WorldSpace size, group=A_Layer_1, parameterName=L1_DetailSize, channelNames {a="", b=Z, g=Y, r=X}, defaultValue=<rgba 64 64 64 0>}
  n56 = VectorParameter() {
   group=A_Layer_1
   parameterName=L1 UVAnim Amount
   channelNames {a="", b=Roughness Anim, g=Emissive-Specular, r=AO Anim}
   defaultValue=<rgba 1 1 1 0>
  }
  n57 = MaterialFunctionCall(
   UVs=NamedRerouteUsage()
   Tobj_Albedo=NamedRerouteUsage()
   Tobj_Normal=NamedRerouteUsage()
   Tobj_AOSR=StaticSwitchParameter(True=NamedRerouteUsage(), False=NamedRerouteUsage()) {desc=Use AOSR instead, group=A_Layer_1, parameterName=L1_UseAOER, sortPriority=255}
   TriplanarAlpha=NamedRerouteUsage()
   AllowTriplanar=StaticSwitchParameter(True=StaticBool() {value=true}, False=StaticBool()) {group=A_Layer_1, parameterName=L1_UseDetailMaps, sortPriority=255}
   TriplanarSize=@n55
   Tobj_DetailNormal=NamedRerouteUsage()
   Tobj_DetailAOSR=NamedRerouteUsage()
   UseAOSRDetail=NamedRerouteUsage()
   UseSER=StaticSwitchParameter(True=StaticBool() {value=true}, False=StaticBool()) {desc="Specular, Emissive, Roughness", group=A_Layer_1, parameterName=L1_UseSER, sortPriority=255}
   UseAOER=StaticSwitchParameter(True=StaticBool() {value=true}, False=StaticBool()) {desc=Use AOSR instead, group=A_Layer_1, parameterName=L1_UseAOER, sortPriority=255}
   Emissive Color=NamedRerouteDeclaration(None=VectorParameter() {group=A_Layer_1, parameterName=L1 Emissive Color}.RGB) {name=L1EmColor_1, nodeColor=<rgba 0.9453 1 0 1>}
   SpecularValue=ScalarParameter() {defaultValue=0.5, group=A_Layer_1, parameterName=L1 Specular, sortPriority=48}
   UseUVAnimation=StaticSwitchParameter(True=StaticBool() {value=true}, False=StaticBool()) {group=A_Layer_1, parameterName=L1_UseUVAnim, sortPriority=255}
   AnimAOSRweights=@n56
   AnimatedUVs=NamedRerouteUsage()
  ) {materialFunction=&MF_LKLayer}
  n58 = GetMaterialAttributes(None=@n57.Result) {attributeGetTypes [#2]}
  n59 = VectorParameter() {group=Vine Color Variation, parameterName=Vine Color Variation Mask R, sortPriority=3, defaultValue=<rgba 0.2781 0.038 0.0077 1>}
  n60 = TextureSampleParameter2D(UVs=Divide(
    A=TextureCoordinate()
    B=ScalarParameter() {defaultValue=0.3, group=Vine Color Variation, parameterName=Color Variation Mask Tiling, sortPriority=2}
   )) {
   group=Vine Color Variation
   parameterName=Color Variation Mask Texture
   samplerType=SAMPLERTYPE_Masks
   sortPriority=1
   texture=/Game/Environments/Helios/Props/Holiday/Holiday_LumpkinKing_A/Textures/Tiling/T_AxisThrow_Vine_Tiling_Mask
  }
  n61 = Multiply(
   A=@n60.R
   B=ScalarParameter() {defaultValue=0.25, group=Vine Color Variation, parameterName=Vine Color Variation R Intensity, sortPriority=4}
  )
  n62 = VectorParameter() {group=Vine Color Variation, parameterName=Vine Color Variation Mask G, sortPriority=5, defaultValue=<rgba 0.1085 0.4531 0.0765 1>}
  n63 = ScalarParameter() {defaultValue=0.1, group=Vine Color Variation, parameterName=Vine Color Variation G Intensity, sortPriority=6}
  n64 = VectorParameter() {group=Vine Color Variation, parameterName=Vine Color Variation Mask B, sortPriority=7, defaultValue=<rgba 0.2058 0.3698 0.005 1>}
  n65 = LinearInterpolate(
   A=LinearInterpolate(A=LinearInterpolate(A=@n58.BaseColor, B=@n59.RGB, Alpha=@n61), B=@n62.RGB, Alpha=Multiply(A=@n60.G, B=@n63))
   B=@n64.RGB
   Alpha=Multiply(A=@n60.B, B=ScalarParameter() {defaultValue=0.2, group=Vine Color Variation, parameterName=Vine Color Variation B Intensity, sortPriority=8})
  )
  NamedRerouteDeclaration(None=StaticSwitchParameter(
    True=SetMaterialAttributes!(MaterialAttributes=@n58.MaterialAttributes, BaseColor=@n65) {attributeSetTypes [#2]}
    False=@n57.Result
   ) {group=Vine Color Variation, parameterName=Use Color Varitation Mask}) {name=L1_MA, nodeColor=<rgba 0 0.8125 1 1>}
  NamedRerouteDeclaration(None=TextureObjectParameter() {group=A_Layer_1, parameterName=L1_AOSR, sortPriority=16, texture=&DefaultTexture}) {name=L1AOSR, nodeColor=<rgba 0.6172 1 0 1>}
  NamedRerouteDeclaration(None=StaticSwitchParameter(True=TextureCoordinate() {coordinateIndex=1}, False=TextureCoordinate()) {desc=Use UV0 otherwise..., group=A_Layer_1, parameterName=L1_UseUV1, sortPriority=0}) {name=L1UVIndex, nodeColor=<rgba 0.7344 1 0 1>}
  n66 = GetMaterialAttributes(None=NamedRerouteUsage()) {attributeGetTypes [#3, #4, #1]}
  n67 = GetMaterialAttributes(None=NamedRerouteUsage()) {attributeGetTypes [#3, #4, #1]}
  n68 = Multiply(A=@n66.Roughness, B=@n67.Roughness)
  n69 = Multiply(A=@n66.AmbientOcclusion, B=@n67.AmbientOcclusion)
  n70 = Add(A=@n66.EmissiveColor, B=@n67.EmissiveColor)
  NamedRerouteDeclaration(None=StaticSwitchParameter(
    True=SetMaterialAttributes!(MaterialAttributes=@n66.MaterialAttributes, Roughness=@n68, AmbientOcclusion=@n69, EmissiveColor=@n70) {attributeSetTypes [#3, #4, #1]}
    False=NamedRerouteUsage()
   ) {group=Options, parameterName=Use Second Layer}) {name=blendedMA, nodeColor=<rgba 0 0.2969 1 1>}
  n71 = TextureObjectParameter() {group=A_Layer_1, parameterName=L1_DetailNormals, samplerType=SAMPLERTYPE_Normal, sortPriority=64, texture=&FlatNormal}
  NamedRerouteDeclaration(None=@n71) {name=L1DNormals, nodeColor=<rgba 1 0 0.9375 1>}
  NamedRerouteDeclaration(None=TextureObjectParameter() {group=A_Layer_1, parameterName=L1_DetailAOSR, sortPriority=64, texture=&DefaultTexture}) {name=L1DAOSR, nodeColor=<rgba 0 1 0.7422 1>}
  NamedRerouteDeclaration(None=TextureObjectParameter() {group=A_Layer_1, parameterName=L1_AOER, sortPriority=16, texture=&DefaultTexture}) {name=L1AOER, nodeColor=<rgba 0.2026 0.2461 1 1>}
  NamedRerouteDeclaration(None=@n4.ThornsParam) {name=thornsGradient_1, nodeColor=<rgba 1 0.2344 0 1>}
  NamedRerouteUsage()
  n72 = Transform(None=Constant3Vector() {constant=<rgba 0 0 1 0>}) {transformSourceType=TRANSFORMSOURCE_Camera}
  n73 = Transform(None=Constant3Vector() {constant=<rgba 0 1 0 0>}) {transformSourceType=TRANSFORMSOURCE_Camera}
  n74 = ObjectPositionWS()
  n75 = Multiply(A=ScalarParameter() {defaultValue=1, group=E_InnerGlowFalloff, parameterName=Lumpkin_InnerGlowSize}, B=ObjectRadius())
  Power(
   Base=SphereMask(
    A=MaterialFunctionCall(Plane Normal Axis=@n72, Plane X Axis=@n73, Plane Center=@n74, UV Scale=@n75) {materialFunction=&VirtualPlaneCoordinates}.UVs
    B=Constant2Vector() {g=0.5, r=0.5}
   ) {attenuationRadius=1, hardnessPercent=0}
   Exp=@n39
  )
  SetMaterialAttributes!(MaterialAttributes, BaseColor=GetMaterialAttributes(None=@n57.Result) {attributeGetTypes [#2]}.BaseColor) {attributeSetTypes [#2]}
  StaticSwitchParameter(True=NamedRerouteUsage(), False=NamedRerouteDeclaration(None=@n5) {name=curveBlendsMA, nodeColor=<rgba 1 0 0.2812 1>}) {group=Options, parameterName=UseEmissiveFresnel, sortPriority=50}
  StaticSwitchParameter(
   True=StaticSwitchParameter(True=@n10, False=@n5) {desc=Use AOSR instead, group=Options, parameterName=UseThornsEmissiveFix, sortPriority=255}
   False=@n5
  ) {desc=Use AOSR instead, group=A_Layer_1, parameterName=L1_UseAOER, sortPriority=255}
  TextureSample(Tex=NamedRerouteUsage()) {texture=/Game/Environments/Helios/Props/Holiday/Holiday_LumpkinKing_A/Textures/Head/T_LumpkinKing_Head_basecolor}
 ]}
