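# Masked, two-sided material for particle mesh versions of the Cube Growth on Deimos rift spawners.
# Cube noise and red vertex color drive dirt-to-base color, a roughness-like mask, detail normal strength and an emissive gradient tinted by particle color.
# A reflection-vector fake specular feeds color and emissive, switched off at low quality.
# Particle alpha thresholds a smooth noise, offset by a random dynamic parameter, for a dissolve mask.
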Material /Game/Athena/Deimos/Spawners/RiftSpawners/Assets/CubeGrowth/M_CubeGrowth_FXMeshes
graph {
 properties {
  bUsedWithMeshParticles=true
  bUsedWithNiagaraMeshParticles=true
  blendMode=BLEND_Masked
  ditheredLODTransition=true
  pixelDepthOffsetMode=PDOM_Legacy
  twoSided=true
  usedShadingModels=Default Lit
 }
 nodes [
  n1 = ReflectionVectorWS()
  n2 = CameraVectorWS()
  n3 = TextureCoordinate()
  n4 = Multiply(A=@n3, B=ScalarParameter() {defaultValue=2, parameterName=CubeNoiseTiling})
  n5 = TextureSample(UVs=@n4) {samplerType=SAMPLERTYPE_LinearColor, texture=/Game/Athena/Environments/Island/CubeGrowth/T_CubeNoise01_H}
  n6 = OneMinus(None=@n5.G)
  n7 = Power(Base=VertexColor().R) {constExponent=4}
  n8 = TextureSample(UVs=@n3) {samplerType=SAMPLERTYPE_Masks, texture=/Game/Global/Textures/Default/Noise/T_CubeNoise}
  n9 = LinearInterpolate(A=ScalarParameter() {defaultValue=0.1, parameterName=R1}, B=ScalarParameter() {defaultValue=0.6, parameterName=R2}, Alpha=@n8.R)
  n10 = Clamp(None=Add(A=@n6, B=Add(A=@n7, B=@n9)) {constB=0})
  n11 = Power(
   Base=Clamp(None=DotProduct(A=@n1, B=@n2))
   Exp=LinearInterpolate(A=ScalarParameter() {defaultValue=40.2714, parameterName=FakeSpecPower}, B=Constant() {r=1}, Alpha=@n10)
  )
  n12 = Power(Base=OneMinus(None=@n10)) {constExponent=4}
  n13 = QualitySwitch(
   Default=Multiply(A=Multiply(A=@n11, B=@n12), B=VectorParameter() {parameterName=FakeSpecularColor, defaultValue=<rgba 0.9823 0.6133 1 1>}.RGB)
   Low=Constant()
  )
  Add(
   A=@n13
   B=LinearInterpolate(
    A=VectorParameter() {parameterName=DirtColor, defaultValue=<rgba 0.1771 0.1291 0.1678 1>}.RGB
    B=VectorParameter() {parameterName=BaseColor, defaultValue=<rgba 0.0326 0.0049 0.07 1>}.RGB
    Alpha=@n5.G
   )
  )
  n14 = Add(A=TextureCoordinate() {uTiling=1.5, vTiling=1.5}, B=DynamicParameter() {paramNames [Random,Param2,Param3,Param4]}.Random)
  Floor(None=Add(A=TextureSample(UVs=@n14) {texture=&T_Smooth_tile}.G, B=ParticleColor().A))
  n15 = VertexColor()
  n16 = Multiply(A=@n7, B=OneMinus(None=@n8.R))
  n17 = Power(Base=Add(A=Multiply(A=@n15.R, B=@n6), B=Multiply(A=@n16) {constB=0.6})) {constExponent=1.5}
  n18 = Multiply(
   A=LinearInterpolate(
    A=VectorParameter() {parameterName=E1, defaultValue=<rgba 0.0386 0.0154 0.1 1>}.RGB
    B=VectorParameter() {parameterName=E2, defaultValue=<rgba 3.5049 0 4 1>}.RGB
    Alpha=Saturate(None=@n17)
   )
   B=ScalarParameter() {defaultValue=0.5, parameterName=EmissiveMasterBrightness}
  )
  n19 = ParticleColor()
  MaterialFunctionCall(Emissive=Add(A=Multiply(A=@n18, B=@n19.RGB), B=@n13)) {materialFunction=&MF_FX_EmissiveAdjustments}
  n20 = TextureSample(UVs=@n4) {samplerType=SAMPLERTYPE_Normal, texture=/Game/Athena/Environments/Island/CubeGrowth/T_CubeNoise01}
  n21 = ScalarParameter() {defaultValue=0.4, parameterName=CubeNoiseNormal}
  n22 = Multiply(A=Constant3Vector() {constant=<rgba 1 1 0 0>}, B=LinearInterpolate(A=@n21, B=Multiply(A=@n21) {constB=3}, Alpha=VertexColor().R))
  Reroute(None=Add(
    A=TextureSample() {samplerType=SAMPLERTYPE_Normal, texture=/Game/Athena/Environments/Island/CubeGrowth/T_CubeGrowth_normals}.RGB
    B=Multiply(A=@n20.RGB, B=@n22)
   ))
  ScalarParameter() {defaultValue=0.4, parameterName=Specular}
 ]
}
