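# Material function that adds animated tentacle World Position Offset to incoming material attributes, gated by the bUseTentacleMovement static switch.
# MF_CubeGrowthTimings drives a height-offset spawn animation staggered by vertex blue, plus a hit reaction that amplifies and shifts the wriggle.
# Phase-shifted sine waves, varied by object position and scaled by vertex red, with speed from vertex green, combine with panning curl noise.
# A texture channel masks this motion, excluding the spawn offset.
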MaterialFunction /Game/Weapons/Prototype/ObjectMover/Materials/MF_DeimosTentacleWPO
graph {nodes [
  n1 = GetMaterialAttributes(None=FunctionInput() {inputName=Material, inputType=FunctionInput_MaterialAttributes}) {attributeGetTypes [#1]}
  n2 = ScalarParameter() {defaultValue=0.35, parameterName=SpawnAnimOffset}
  n3 = Add(B=@n2) {constA=1}
  n4 = MaterialFunctionCall() {materialFunction=/Game/Athena/Environments/Island/CubeGrowth/MF_CubeGrowthTimings}
  n5 = VertexColor()
  n6 = Saturate(None=Subtract(A=LinearInterpolate(A=@n3, Alpha=@n4.SpawnTiming) {constA=2, constB=0}, B=LinearInterpolate(B=@n2, Alpha=@n5.B) {constB=0}))
  n7 = Power(Base=@n6, Exp=ScalarParameter() {defaultValue=3, parameterName=SpawnAnimPower}) {constExponent=1.5}
  n8 = VectorParameter() {parameterName=HeightOffset, defaultValue=<rgba 0 0 -1600 1>}
  n9 = ObjectPositionWS()
  n10 = WorldPosition()
  n11 = PreSkinnedPosition()
  n12 = ComponentMask(None=Divide(A=@n11, B=PreSkinnedLocalBounds().Max)) {b=true}
  n13 = ScalarParameter() {defaultValue=5, parameterName=SpawnWiggleSpeed}
  n14 = Add(
   A=MaterialFunctionCall!(In=@n4.HitTiming) {materialFunction=/Game/Effects/Fort_Effects/Materials/_Functions/MF_Smoothstep_EaseOut}.Result
   B=Add(A=Multiply(A=@n6, B=@n13) {constB=7}, B=Time())
  )
  n15 = VertexColor()
  n16 = MaterialFunctionCall(Float3=ObjectPositionWS()) {materialFunction=&BreakOutFloat3Components}
  n17 = Add(A=@n16.R, B=@n16.R)
  n18 = Abs(None=Divide(A=Add(A=@n17, B=@n16.B), B=ScalarParameter() {defaultValue=2048, group=TentacleMovement, parameterName=PositionOffsetBias}))
  n19 = Add(A=Multiply(A=@n15.G, B=ScalarParameter() {defaultValue=1, group=TentacleMovement, parameterName=WiggleOffsetBias}), B=@n18)
  n20 = ScalarParameter() {defaultValue=0.25, group=TentacleMovement, parameterName=MinWiggleSpeed}
  n21 = ScalarParameter() {defaultValue=0.5, group=TentacleMovement, parameterName=MaxWiggleSpeed}
  n22 = Multiply(A=Multiply(A=Add(A=@n14, B=@n19), B=LinearInterpolate(A=@n20, B=@n21, Alpha=@n15.G))) {constB=-1}
  n23 = Add(A=@n12, B=@n22)
  n24 = ScalarParameter() {defaultValue=1.5, group=TentacleMovement, parameterName=XWriggleFrequency}
  n25 = Add(A=Add(A=@n12, B=@n18) {constB=0.35}, B=@n22)
  n26 = ScalarParameter() {defaultValue=1.5, group=TentacleMovement, parameterName=YWriggleFrequency}
  n27 = Add(A=Add(A=@n12, B=@n18) {constB=0.75}, B=@n22)
  n28 = ScalarParameter() {defaultValue=1.5, group=TentacleMovement, parameterName=ZWriggleFrequency}
  n29 = MaterialFunctionCall(X=Sine(None=Multiply(A=@n23, B=@n24) {constB=2}), Y=Sine(None=Multiply(A=@n25, B=@n26)), Z=Sine(None=Multiply(A=@n27, B=@n28))) {materialFunction=&MakeFloat3}
  n30 = VertexColor()
  n31 = Constant() {r=0.2}
  n32 = VectorParameter() {group=TentacleMovement, parameterName=HitDisplacementAmt, defaultValue=<rgba 512 512 512 1>}
  n33 = Power(Base=@n4.HitTiming) {constExponent=0.35}
  n34 = Saturate(None=ConstantBiasScale(None=Sine(None=@n33)))
  n35 = Multiply(
   A=Clamp(None=Add(A=@n30.R, B=@n31))
   B=Add(
    A=VectorParameter() {group=TentacleMovement, parameterName=WiggleDisplacementAmt, defaultValue=<rgba 12 12 12 1>}.RGB
    B=Multiply(A=@n32.RGB, B=@n34)
   )
  )
  n36 = Divide(A=@n11, B=VectorParameter() {group=TentacleMovement, parameterName=WritheNoiseFrequency, defaultValue=<rgba 64 64 128 1>}.RGB)
  n37 = Multiply(A=@n14, B=VectorParameter() {group=TentacleMovement, parameterName=WrithePanSpeedAndDirection, defaultValue=<rgba -0.2 0.2 1 1>}.RGB)
  n38 = ConstantBiasScale(None=VectorNoise(World Position=Add(A=@n36, B=@n37)) {noiseFunction=VNF_CurlALU})
  n39 = VectorParameter() {group=TentacleMovement, parameterName=WritheDispalcement_AMt, defaultValue=<rgba 1 1 1 1>}
  n40 = Add(A=Normalize(VectorInput=Subtract(A=@n9, B=@n10.XYZ)), B=Add(A=Multiply(A=@n29.Result, B=@n35), B=Multiply(A=@n38, B=@n39.RGB)))
  n41 = TextureSampleParameter2D() {
   group=TentacleMovement
   parameterName=WPO_Mask
   samplerType=SAMPLERTYPE_Masks
   texture=/Game/Athena/Deimos/Spawners/RiftSpawners/Assets/CubeGrowth/T_Deimos_Storm_FX
  }
  n42 = ChannelMaskParameter(None=@n41.RGB) {parameterName=WPO_MaskCHannel}
  n43 = Add(A=@n1.WorldPositionOffset, B=Add(A=Multiply(A=@n7, B=@n8.RGB), B=Multiply(A=@n40, B=@n42.R)))
  FunctionOutput(None=StaticSwitchParameter(
    True=SetMaterialAttributes!(MaterialAttributes=@n1.MaterialAttributes, WorldPositionOffset=@n43) {attributeSetTypes [#1]}
    False=@n1.MaterialAttributes
   ) {parameterName=bUseTentacleMovement})
 ]}
