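# Master material for the Deimos rift spawner base, animating its spawn-in with masked rune textures and a magenta glow.
# An exponent-shaped SpawnTiming from MF_CubeGrowthTimings fades the mask-driven emissive and drives world position offset.
# That offset pulls vertices toward a point offset from the object position, staggered by the mask and released as timing rises.
# The attributes then pass through the DeimosTentacleWPO and CreativeMoveTool functions.
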Material /Game/Athena/Deimos/Spawners/RiftSpawners/Assets/CubeGrowth/M_SpawnerBase_Master
graph {
 properties {bUseMaterialAttributes=true, ditheredLODTransition=true, pixelDepthOffsetMode=PDOM_Legacy, usedShadingModels=Default Lit}
 nodes [
  n1 = TextureSampleParameter2D() {parameterName=Mask, samplerType=SAMPLERTYPE_Masks, texture=/Game/Athena/Environments/Island/Runes/Texture/T_RuneGround_02_Emissive_S}
  n2 = Power(
   Base=MaterialFunctionCall() {materialFunction=/Game/Athena/Environments/Island/CubeGrowth/MF_CubeGrowthTimings}.SpawnTiming
   Exp=ScalarParameter() {defaultValue=0.15, parameterName=AnimExponent}
  )
  n3 = Clamp(None=Multiply(A=OneMinus(None=@n2)) {constB=1.5})
  n4 = Constant3Vector() {constant=<rgba 0.9264 0 1 1>}
  n5 = TextureSampleParameter2D() {parameterName=Normal, samplerType=SAMPLERTYPE_Normal, texture=/Game/Athena/Environments/Island/Runes/Texture/T_RuneGround_02_Emissive_N}
  n6 = Add(A=ObjectPositionWS(), B=VectorParameter() {defaultValue=<rgba 0 0 -100 1>}.RGB)
  n7 = WorldPosition()
  n8 = Add(A=LinearInterpolate(Alpha=@n1.R) {constA=-0.7, constB=0.7}, B=LinearInterpolate(Alpha=@n2) {constA=1.5, constB=-0.5})
  n9 = MakeMaterialAttributes!(
   BaseColor=TextureSampleParameter2D() {parameterName=Diffuse, texture=/Game/Athena/Environments/Island/Runes/Texture/T_RuneGround_02_Emissive_D}.RGB
   Metallic
   Specular=@n1.R
   Roughness=@n1.R
   Anisotropy
   EmissiveColor=Multiply(A=Multiply(A=@n3, B=@n4), B=@n1.R) {constB=4}
   Opacity
   OpacityMask
   Normal=@n5.RGB
   Tangent
   WorldPositionOffset=Multiply(A=Subtract(A=@n6, B=@n7.XYZ), B=Clamp(None=@n8))
   SubsurfaceColor
   ClearCoat
   ClearCoatRoughness
   AmbientOcclusion
   Refraction
   CustomizedUVs_0
   CustomizedUVs_1
   CustomizedUVs_2
   CustomizedUVs_3
   CustomizedUVs_4
   CustomizedUVs_5
   CustomizedUVs_6
   CustomizedUVs_7
   PixelDepthOffset
   ShadingModel
   Displacement
  )
  MaterialFunctionCall(MA In=MaterialFunctionCall!(Material=@n9) {materialFunction=/Game/Weapons/Prototype/ObjectMover/Materials/MF_DeimosTentacleWPO}.Result) {materialFunction=&MF_CreativeMoveTool}
  n10 = Transform(None=@n5.RGB)
  n11 = VertexNormalWS()
  Multiply(A=Clamp(None=DotProduct(A=@n10, B=@n11)))
  StaticSwitchParameter(False=@n9) {parameterName=isIceandSnow}
 ]
}
